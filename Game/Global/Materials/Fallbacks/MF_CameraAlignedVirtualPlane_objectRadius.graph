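# Material function that produces UVs on a camera-facing virtual plane, scaled by object radius.
# VirtualPlaneCoordinates takes the camera direction as plane normal and doubled pre-skinned local half extents, transformed to world space, as plane center.
# Object radius times the UV Tiling input sets the UV scale, and UV Offset is added to the resulting UVs.
# A Max of bounds and radius is left unconnected.
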MaterialFunction /Game/Global/Materials/Fallbacks/MF_CameraAlignedVirtualPlane_objectRadius
graph {nodes [
  n1 = FunctionInput() {
   bUsePreviewValueAsDefault=true
   description=Preskinned local pos offset
   inputName=UV Offset
   inputType=FunctionInput_Vector2
   sortPriority=1
  }
  n2 = MaterialFunctionCall() {materialFunction=&CameraDirectionVector}
  n3 = PreSkinnedLocalBounds()
  n4 = TransformPosition(None=Add(A=@n3.Half Extents, B=@n3.Half Extents)) {transformType=TRANSFORMPOSSOURCE_World}
  n5 = ObjectRadius()
  n6 = FunctionInput() {bUsePreviewValueAsDefault=true, description=UV scale factor, inputName=UV Tiling, inputType=FunctionInput_Vector2, previewValue=<1 1 0 1>}
  n7 = Multiply(A=@n5, B=@n6)
  FunctionOutput(None=Add(
    A=@n1
    B=MaterialFunctionCall(Plane Normal Axis=@n2.Result, Plane Center=@n4, UV Scale=@n7) {materialFunction=&VirtualPlaneCoordinates}.UVs
   ))
  Max(A=@n3.Max, B=@n5)
 ]}
