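# Material function supplying the blue fallback base look for the character death (derez) transition.
# A rim input, Fresnel by default, blends two derez color parameters into emissive over a black, glossy surface, keeping the base material's normal.
# Epic quality reshapes the rim and adds a sine pulse over time.
# A desaturated animated grid on a camera-aligned plane, panning in opposite directions on alternating rows, is declared as a named reroute.
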MaterialFunction /Game/Global/Materials/Fallbacks/MF_Derez_BaseBlue
graph {
 properties {description=Base Colors for the death Transition}
 nodes [
  Constant3Vector() {constant=<rgba 0 3.8199 6 1>}
  Constant3Vector() {constant=<rgba 0.0583 0.3109 0.4 1>}
  n1 = Constant()
  n2 = FunctionInput(Preview=Fresnel() {baseReflectFraction=0, exponent=1}) {bUsePreviewValueAsDefault=true, inputName=Rim, inputType=FunctionInput_Scalar}
  n3 = Clamp(None=@n2)
  n4 = Power(Base=OneMinus(None=@n3)) {constExponent=4}
  n5 = LinearInterpolate(Alpha=Power(Base=@n3)) {constA=0.3, constB=0.9}
  n6 = QualitySwitch(Default=@n2, Low=@n2, High=@n2, Medium=@n2, Epic=Multiply(A=Add(A=@n4, B=@n5)) {constB=0.6})
  n7 = Add(
   A=VectorParameter() {group=Derez, parameterName=Character_Derez_Color_01, defaultValue=<rgba 0.0583 0.3109 0.4 1>}.RGB
   B=Multiply(
    A=Power(Base=@n6) {constExponent=3}
    B=VectorParameter() {group=Derez, parameterName=Character_Derez_Color_02, defaultValue=<rgba 0 3.8199 6 1>}.RGB
   )
  )
  n8 = NamedRerouteUsage()
  n9 = Time()
  n10 = ConstantBiasScale(None=Sine(None=Add(A=@n8, B=@n9)))
  n11 = Multiply(A=Power(Base=@n7), B=Multiply(A=LinearInterpolate(Alpha=@n10) {constA=0.6}, B=NamedRerouteUsage()))
  n12 = Multiply(A=@n7) {constB=0.3}
  n13 = GetMaterialAttributes(None=FunctionInput(Preview=SetMaterialAttributes!(MaterialAttributes)) {bUsePreviewValueAsDefault=true, inputName=BaseMaterial, inputType=FunctionInput_MaterialAttributes}) {attributeGetTypes [#1, #2]}
  n14 = SetMaterialAttributes!(
   MaterialAttributes
   BaseColor=@n1
   Metallic=@n1
   Specular=Constant() {r=1}
   Roughness=Constant() {r=0.2}
   EmissiveColor=QualitySwitch(Default=@n7, Low=@n7, High=@n7, Medium=@n7, Epic=Add(A=@n11, B=@n12))
   Opacity
   OpacityMask
   Anisotropy
   Normal=@n13.Normal
   WorldPositionOffset=@n13.Normal
   WorldDisplacement
   TessellationMultiplier
   SubsurfaceColor=Constant()
   CustomData0=Constant()
  ) {attributeSetTypes [
    #3
    #4
    #5
    #6
    #7
    #8
    #9
    #10
    #1
    #2
    #11
    #12
    #13
    #14
   ]}
  FunctionOutput(None=@n14)
  MakeMaterialAttributes!(
   BaseColor
   Metallic
   Specular
   Roughness
   Anisotropy
   EmissiveColor
   Opacity
   OpacityMask
   Normal
   Tangent
   WorldPositionOffset
   SubsurfaceColor
   ClearCoat
   ClearCoatRoughness
   AmbientOcclusion
   Refraction
   CustomizedUVs_0
   CustomizedUVs_1
   CustomizedUVs_2
   CustomizedUVs_3
   CustomizedUVs_4
   CustomizedUVs_5
   CustomizedUVs_6
   CustomizedUVs_7
   PixelDepthOffset
   ShadingModel
   Displacement
  )
  NamedRerouteDeclaration(None=ScalarParameter() {defaultValue=15, group=Derez, parameterName=FallbackGridTiling}) {name=FallbackTiling, nodeColor=<rgba 0.8281 1 0 1>}
  n15 = MaterialFunctionCall!(UV Tiling, UV Offset) {materialFunction=&/MF_CameraAlignedVirtualPlane_objectRadius}
  n16 = VectorParameter() {group=Derez, parameterName=FallbackGridPannerSpeed, defaultValue=<rgba 0.08 0 0 1>}
  n17 = Multiply(A=@n16.RGB) {constB=-0.8}
  n18 = Multiply(
   A=MaterialFunctionCall() {materialFunction=&MF_CameraAlignedVirtualPlane_objectRadius}.Result
   B=NamedRerouteUsage()
  ) {constB=20}
  n19 = Fmod(A=ComponentMask(None=Ceil(None=@n18)) {g=true}, B=Constant() {r=2})
  n20 = LinearInterpolate(
   A=MaterialFunctionCall!(
    UV=Panner(Coordinate=@n15.Result, Speed=@n16.RGB) {speedX=0.1}
    Tiling=NamedRerouteUsage()
    Tiling X
    Tiling Y
    Color 01
    Color 02
    Speed
    Speed Random Amount
    Line Thickness
    Line Hardness
    Line Color
   ) {materialFunction=/Game/Global/MaterialFunctions/Cosmetics/AnimatedGrid/MF_AnimatedGrid}."Result (V3)"
   B=MaterialFunctionCall(UV=Panner(Coordinate=@n15.Result, Speed=@n17) {speedX=-0.1}, Tiling=NamedRerouteUsage()) {materialFunction=&MF_AnimatedGrid}."Result (V3)"
   Alpha=@n19
  )
  n21 = LinearInterpolate(Alpha=@n19) {constA=0.6}
  NamedRerouteDeclaration(None=Desaturation(None=Multiply(A=@n20, B=@n21), Fraction=Constant() {r=1})) {name=Grid, nodeColor=<rgba 1 0 0.1172 1>}
 ]
}
