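# M_DeRezShader is a masked, Default Lit material in the global Fallbacks folder, apparently a stand-in de-rez look for meshes.
# The MF_Derez_BaseBlue function supplies the effect as material attributes, with its BaseMaterial and Rim inputs unconnected.
# A Get/Set MaterialAttributes pair passes that result through, rewiring the Emissive Color slot.
# Usage flags cover skeletal, clothing, morph target, mesh deformer and static lighting, with automatic usage setting disabled.
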Material /Game/Global/Materials/Fallbacks/M_DeRezShader
graph {
 properties {
  bAutomaticallySetUsageInEditor=false
  bUseMaterialAttributes=true
  bUsedWithClothing=true
  bUsedWithMeshDeformer=true
  bUsedWithMorphTargets=true
  bUsedWithSkeletalMesh=true
  bUsedWithStaticLighting=true
  blendMode=BLEND_Masked
  pixelDepthOffsetMode=PDOM_Legacy
  translucencyLightingMode=TLM_VolumetricDirectional
  usedShadingModels=Default Lit
 }
 nodes [
  n1 = GetMaterialAttributes(None=MaterialFunctionCall!(BaseMaterial, Rim) {materialFunction=&/MF_Derez_BaseBlue}.Result) {attributeGetTypes [#1]}
  SetMaterialAttributes(MaterialAttributes=@n1.MaterialAttributes, Emissive Color=@n1.MaterialAttributes) {attributeSetTypes [#1]}
 ]
}
